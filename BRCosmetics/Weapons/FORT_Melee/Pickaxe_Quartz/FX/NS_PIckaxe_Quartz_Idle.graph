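# Idle Niagara effect for the Quartz pickaxe, emitting intermittent, short-lived sprite blobs that follow the weapon's skeletal mesh.
# Two similar emitters spawn probabilistically and re-sample User.Skeletal Mesh each update, with curve-driven upward offsets; the more frequent one also randomizes X offset.
# Upward velocity is scaled down over life, while curves grow and shrink sprite size and drive dynamic material parameters.
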NiagaraSystem /BRCosmetics/Weapons/FORT_Melee/Pickaxe_Quartz/FX/NS_PIckaxe_Quartz_Idle
dependencies {
 usedDataInterfaces [&NiagaraDataInterfaceCurve]
 usedDynamicInputs [&Lerp_Float]
 usedModules [&SystemState]
}
emitters {
 InitialBlob001 {
  emitterUpdateScript [
   &EmitterState()
   &SpawnRate(Spawn Probability=0.22, SpawnRate=1.5, Use Spawn Probability=true)
  ]
  particleSpawnScript [
   &InitializeParticle(
    Lifetime Max=1
    Lifetime Min=0.75
    Lifetime Mode=::NewEnumerator1
    Mass=&RandomRangeFloat(Minimum=0.1, Maximum=0.25)
    Mass Mode=::NewEnumerator1
    UsePositionOffset=false
    Write Material Random=true
   )
   /Niagara/Modules/Update/Mass/CalculateSizeAndRotationalInertiaByMass(Density by Material Type:ENiagaraMassByVolume=::NewEnumerator2, Mass Modulation=100, Module.Manually Enter Density=false)
   &AddVelocity(Rotation Coordinate Space=::World, Velocity=&RandomRangeVector(Minimum=<0 0 35>, Maximum=<0 0 50>))
  ]
  particleUpdateScript [
   &ParticleState()
   &SystemLocation(MaskSpawnGroup=false)
   &SkeletalMeshLocation(
    Bone Sampling Mode=::NewEnumerator4
    MaskSpawnGroup=false
    Sampled Position Offset=&Lerp_Vector(
     A=Linked() {linkedVariable {name=Particles.Position, type=NiagaraPosition}}
     B=&AddVectorToPosition(Position=$Particles.Position, Vector=&RandomRangeVector(Minimum=<0 0 5>, Maximum=<0 0 15>))
     Alpha=&FloatFromCurve(FloatCurve=DataInterface() {propertyValues {Curve {keys [{tangentMode=RCTM_User, time=0, value=0}, {tangentMode=RCTM_User, time=1, value=1}]}}})
    )
    Sampled Position Offset Coordinate Space:ENiagaraSkelMeshTransforms=::NewEnumerator0
    Skeletal Mesh=$User.Skeletal Mesh
   ) {module=SkeletalMeshLocation001}
   &ScaleSpriteSize(Uniform Curve Sprite Scale=DataInterface() {propertyValues {Curve {keys [
        {interpMode=RCIM_Cubic, tangentMode=RCTM_User, time=0, value=0}
        <at 0.2 1>
        {interpMode=RCIM_Cubic, tangentMode=RCTM_User, time=0.5, value=1}
        {arriveTangent=-3.73, interpMode=RCIM_Cubic, leaveTangent=-3.73, tangentMode=RCTM_User, time=1, value=0}
       ]}}})
   &DynamicMaterialParameters(
    Index 0 Param 1=&FloatFromCurve(FloatCurve=DataInterface() {propertyValues {Curve {keys [
         {arriveTangent=0.0167, interpMode=RCIM_Cubic, leaveTangent=0.0167, tangentMode=RCTM_User, time=0, value=0}
         <at 0.25 1 RCIM_Cubic>
         <at 1 0 RCIM_Cubic>
        ]}}})
    Index 0 Param 2=5
    Index 0 Param 3=15
    Index 0 Param 4=6
    Index 1 Param 2=&RandomRangeFloat(Minimum=1.25, Maximum=2.5)
    Module.Param0WriteEnabled=true
    Module.Param0WriteEnabled1=false
    Module.Param1WriteEnabled=true
    Module.Param1WriteEnabled1=true
    Module.Param2WriteEnabled=true
    Module.Param2WriteEnabled1=true
    Module.Param3WriteEnabled=true
    Module.Param3WriteEnabled1=true
    Write Parameter Index 1=true
   )
   &ScaleVelocity(Velocity Scale=&Multiply_VectorByFloat(
     Vector=<1 1 1>
     Float=&FloatFromCurve(FloatCurve=DataInterface() {propertyValues {Curve {keys [<at 0 1>,<at 1 0>]}}})
    ))
   &SolveForcesAndVelocity(Clamp Velocity=false, Limit Acceleration=false)
  ]
  rendererClasses [&NiagaraSpriteRendererProperties]
  renderers [{rendererClass=&NiagaraSpriteRendererProperties}]
 }
 InitialBlob002 {
  emitterUpdateScript [
   &EmitterState()
   &SpawnRate(Spawn Probability=0.22, SpawnRate=4, Use Spawn Probability=true)
  ]
  particleSpawnScript [
   &InitializeParticle(
    Lifetime Max=1
    Lifetime Min=0.75
    Lifetime Mode=::NewEnumerator1
    Mass=&RandomRangeFloat(Minimum=0.1, Maximum=0.25)
    Mass Mode=::NewEnumerator1
    UsePositionOffset=false
    Write Material Random=true
   )
   &CalculateSizeAndRotationalInertiaByMass(Density by Material Type:ENiagaraMassByVolume=::NewEnumerator2, Mass Modulation=100, Module.Manually Enter Density=false)
   &AddVelocity(Rotation Coordinate Space=::World, Velocity=&RandomRangeVector(Minimum=<0 0 35>, Maximum=<0 0 50>))
  ]
  particleUpdateScript [
   &ParticleState()
   &SkeletalMeshLocation(
    Bone / Socket Index=1
    Bone Sampling Mode=::NewEnumerator4
    MaskSpawnGroup=false
    Position Sampling=::NewEnumerator1
    Sampled Position Offset=&AddVectorToPosition(
     Position=&MakeVector(X=&RandomRangeFloat(Minimum=10, Maximum=-10), Y=0, Z=0)
     Vector=&MakeVector(
      X=0
      Y=0
      Z=&Multiply_Float(
       A=&RandomRangeFloat(Minimum=5, Maximum=15)
       B=&FloatFromCurve(FloatCurve=DataInterface() {propertyValues {Curve {keys [
            {interpMode=RCIM_Cubic, tangentMode=RCTM_User, time=0, value=0}
            {arriveTangent=1.8078, interpMode=RCIM_Cubic, leaveTangent=1.8078, tangentMode=RCTM_User, time=1, value=1}
           ]}}})
      )
     )
    )
    Skeletal Mesh=$User.Skeletal Mesh
   ) {module=SkeletalMeshLocation001}
   &ScaleSpriteSize(Uniform Curve Sprite Scale=DataInterface() {propertyValues {Curve {keys [
        {interpMode=RCIM_Cubic, tangentMode=RCTM_User, time=0, value=0}
        <at 0.2 1>
        {interpMode=RCIM_Cubic, tangentMode=RCTM_User, time=0.5, value=1}
        {arriveTangent=-3.73, interpMode=RCIM_Cubic, leaveTangent=-3.73, tangentMode=RCTM_User, time=1, value=0}
       ]}}})
   &DynamicMaterialParameters(
    Index 0 Param 1=&FloatFromCurve(FloatCurve=DataInterface() {propertyValues {Curve {keys [
         {arriveTangent=0.0167, interpMode=RCIM_Cubic, leaveTangent=0.0167, tangentMode=RCTM_User, time=0, value=0}
         <at 0.25 1 RCIM_Cubic>
         <at 1 0 RCIM_Cubic>
        ]}}})
    Index 0 Param 2=5
    Index 0 Param 3=15
    Index 0 Param 4=6
    Index 1 Param 2=&RandomRangeFloat(Minimum=1.25, Maximum=2.5)
    Module.Param0WriteEnabled=true
    Module.Param0WriteEnabled1=false
    Module.Param1WriteEnabled=true
    Module.Param1WriteEnabled1=true
    Module.Param2WriteEnabled=true
    Module.Param2WriteEnabled1=true
    Module.Param3WriteEnabled=true
    Module.Param3WriteEnabled1=true
    Write Parameter Index 1=true
   )
   &ScaleVelocity(Velocity Scale=&Multiply_VectorByFloat(
     Vector=<1 1 1>
     Float=&FloatFromCurve(FloatCurve=DataInterface() {propertyValues {Curve {keys [<at 0 1>,<at 1 0>]}}})
    ))
   &SolveForcesAndVelocity(Clamp Velocity=false, Limit Acceleration=false)
  ]
  rendererClasses [&NiagaraSpriteRendererProperties]
  renderers [{rendererClass=&NiagaraSpriteRendererProperties}]
 }
}
summary {
 systemName=NS_PIckaxe_Quartz_Idle
 emitters [
  {emitterName=InitialBlob002, rendererClasses [&NiagaraSpriteRendererProperties]}
  {emitterName=InitialBlob001, rendererClasses [&NiagaraSpriteRendererProperties]}
 ]
 userVariables [User.Skeletal Mesh {
   type=NiagaraDataInterfaceSkeletalMesh
   defaultValue {
    struct=NiagaraExt_VariableValue_DataInterface
    value {dataInterface=&self:NiagaraDataInterfaceSkeletalMesh_0}
   }
  }]
}
