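# Physics asset for the Quartz pickaxe, a two-handed melee weapon cosmetic in BRCosmetics.
# Simple capsule and box bodies cover the handle, blade, a floating piece and two stalks, with every dyn_ joint constrained directly to the handle bone.
# Stalk constraints lock swing2 and twist, leaving swing1 as the bending axis, and each stalk-to-handle pair appears twice.
# Blade and float constraints carry very tight angular limit values.
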PhysicsAsset /BRCosmetics/Weapons/FORT_Melee/Pickaxe_Quartz/Meshes/Quartz_Axe_Physics
bodies {
 Melee_TwoHanded_Handle [
  {shapeType=Capsule, length=73.9253, radius=12.9553, center=<-2.7624 -1.9301 28.636>, rotation=<rot 4.0689 -90 5.6237>}
  {
   shapeType=Capsule
   length=15.2833
   radius=2.1526
   shapeName=Melee_TwoHanded_Handle_capsule
   center=<-0.4097 -7.0363 91.1757>
   rotation=<rot -0.6408 -0.3936 -115.5948>
  }
 ]
 dyn_blade_joint [{
   shapeType=Box
   extentX=7.5
   extentY=70.089
   extentZ=68.9697
   shapeName=dyn_blade_joint_box
   center=<0 0.6998 -4.1457>
   rotation=<rot 0 0 -9.581>
  }]
 dyn_float_2_joint [{
   shapeType=Capsule
   length=18.7961
   radius=7.5297
   shapeName=dyn_float_2_joint_capsule
   center=<-5.2929 -0.2091 1.0572>
   rotation=<rot 0 -89.7283 -68.2697>
  }]
 dyn_stalk_1_1_joint [
  {shapeType=Capsule, length=10.8259, radius=1.1205, center=<16.8872 -0.0548 -6.9376>, rotation=<rot 0.6096 86.6938 -94.5942>}
  {
   shapeType=Capsule
   length=14.6053
   radius=2.2033
   shapeName=dyn_stalk_1_1_joint_capsule
   center=<37.2644 -1.3395 1.0035>
   rotation=<rot -27.7562 -9.9674 10.1657>
  }
 ]
 dyn_stalk_2_1_joint [{shapeType=Capsule, length=35.3806, radius=1.6248, center=<28.4871 -0.8276 -2.705>, rotation=<rot 0.2828 86.6036 -67.2732>}]
}
constraints [
 {
  bone1Name=dyn_stalk_2_1_joint
  bone2Name=Melee_TwoHanded_Handle
  swing1LimitDegrees=20
  swing2LimitDegrees=8
  swing2Motion=Locked
  twistLimitDegrees=45
  twistMotion=Locked
 }
 {
  bone1Name=dyn_stalk_1_1_joint
  bone2Name=Melee_TwoHanded_Handle
  swing1LimitDegrees=20
  swing2LimitDegrees=8
  swing2Motion=Locked
  twistLimitDegrees=45
  twistMotion=Locked
 }
 {
  bone1Name=dyn_stalk_1_1_joint
  bone2Name=Melee_TwoHanded_Handle
  swing1LimitDegrees=20
  swing2LimitDegrees=8
  swing2Motion=Locked
  twistLimitDegrees=45
  twistMotion=Locked
 }
 {
  bone1Name=dyn_stalk_2_1_joint
  bone2Name=Melee_TwoHanded_Handle
  swing1LimitDegrees=15
  swing2LimitDegrees=8
  swing2Motion=Locked
  twistLimitDegrees=45
  twistMotion=Locked
 }
 {bone1Name=dyn_float_2_joint, bone2Name=Melee_TwoHanded_Handle, swing1LimitDegrees=1, swing2LimitDegrees=1, twistLimitDegrees=1}
 {bone1Name=dyn_blade_joint, bone2Name=Melee_TwoHanded_Handle, swing1LimitDegrees=1, swing2LimitDegrees=1, twistLimitDegrees=1}
]
