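# Animation Blueprint for the Quartz Axe pickaxe that animates its floating elements and adds physics-driven secondary motion.
# Real time drives a continuous roll and yaw spin on one bone and an offset sine bob on another, both applied additively.
# A RigidBody node then simulates with the axe's physics asset, its alpha easing to half while sprinting and back to full otherwise.
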AnimBlueprint /BRCosmetics/Weapons/FORT_Melee/Pickaxe_Quartz/Meshes/Quartz_Axe_AnimBP
target_skeleton=&/Quartz_Axe_Skeleton
graphs [
 AnimGraph {nodes [
   n1 = "Transform (Modify) Bone - Bone: dyn_float_1_joint"(
    ComponentPose=Local To Component(LocalPose=Input Pose())
    Alpha=1
    Translation="0,0,0"
    Rotation=Get rotator_element_A() -> (rotator_element_A:Rotator)
    Scale="1,1,1"
   ) {
    rotationMode=BMM_Additive
    rotationSpace=BCS_BoneSpace
    translationMode=BMM_Additive
    alphaBoolBlend {blendOption=Linear}
    boneToModify {boneName=dyn_float_1_joint}
   }
   n2 = RigidBody(
    ComponentPose="Transform (Modify) Bone - Bone: dyn_float_2_joint"(ComponentPose=@n1, Alpha=1, Translation=Get element_B_float() -> (element_B_float:Vector), Rotation="0,0,0", Scale="1,1,1") {translationMode=BMM_Additive, alphaBoolBlend {blendOption=Linear}, boneToModify {boneName=dyn_float_2_joint}}
    Alpha=Get RigidBody_alpha() -> (RigidBody_alpha:double)
   ) {
    bClampLinearTranslationLimitToRefPose=true
    bOverrideWorldGravity=true
    overridePhysicsAsset=&/Quartz_Axe_Physics
    alphaBoolBlend {blendOption=Linear}
    componentAppliedLinearAccClamp=<3000 3000 3000>
    componentLinearAccScale=<0.3 0.3 0.3>
    componentLinearVelScale=<0.3 0.3 0.3>
   }
   Output Pose(Result=Component To Local(ComponentPose=@n2)) {layerGroup=None}
  ]}
 EventGraph {nodes [
   n1 = Event BlueprintUpdateAnimation()
   > n2 = Sequence!(execute) -> (then_0, then_1, then_2:Exec, then_3:Exec)
   n3 = float * float(A=float / float(A=GetRealTimeSeconds(), B=-720) {is=title K2Node_CallFunction}, B=25000)
   > Set rotator_element_A(rotator_element_A:Rotator=MakeRotator(Roll=@n3, Yaw=@n3)) -> (Output_Get:Rotator)
   n4 = GetRealTimeSeconds()
   n5 = float * float(A=1, B=0.5)
   n6 = float * float(A="Sin (Radians)"(A=float * float(A=@n5, B=@n4)), B=0.5)
   n7 = float * float(A=1, B=0.5)
   n8 = float * float(A="Sin (Radians)"(A=float * float(A=@n7, B=@n4)), B=0.7)
   n9 = float * float(A=1, B=0.5)
   n10 = float * float(A="Sin (Radians)"(A=float * float(A=@n9, B=@n4)), B=0.5)
   Set element_B_float(
    execute=@n2.then_1
    element_B_float:Vector=vector + vector!(A=MakeVector(X=@n10, Y=@n6, Z=@n8), B_X:double=0, B_Y:double=4, B_Z:double=0) -> (ReturnValue)
   ) -> (Output_Get:Vector)
   n11 = Branch(execute=@n2.then_2, Condition=Get bIsSprinting())
   n12 = Get RigidBody_alpha()
   > Set RigidBody_alpha(RigidBody_alpha:double=FInterpTo(Current=@n12, Target=0.5, DeltaTime=@n1.DeltaTimeX, InterpSpeed=3)) -> (Output_Get:double)
   Set RigidBody_alpha(execute=@n11.else, RigidBody_alpha=FInterpTo(Current=@n12, Target=1, DeltaTime=@n1.DeltaTimeX, InterpSpeed=3))
  ]}
]
variables [
 element_B_float {type=Vector}
 RigidBody_alpha {type=number}
 rotator_element_A {type=Rotator}
]
